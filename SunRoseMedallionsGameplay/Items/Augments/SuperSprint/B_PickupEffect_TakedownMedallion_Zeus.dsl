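; Pickup effect Blueprint for the Zeus takedown medallion, toggling its idle visual as the item is picked up and tossed.
; On BeginPlay it runs the parent logic and destroys the inherited or local BG component.
; Pickup hides the NS_TDM_Pickup_Idle component and tossing shows it again, propagating to children, after calling the parent handlers.
; Tick, overlap and the construction script add no logic of their own.
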(event Collision|EventActorBeginOverlap
  (params (OtherActor :type Actor%20Object%20Reference)))

(event EventTick
  (params (DeltaSeconds :type Float%20%28single-precision%29)))

(event EventBeginPlay
  (|CallParentFunction)
  (Components|DestroyComponent (Variables|Default|GetBG)))

(event EventOnPickedUp
  (params (PickupTarget :type Fort%20Pawn%20Object%20Reference))
  (|CallParentFunction PickupTarget)
  (Development|SetHiddeninGame (Variables|Default|GetNS_TDM_Pickup_Idle) true true))

(event EventOnTossed
  (|CallParentFunction)
  (Development|SetHiddeninGame (Variables|Default|GetNS_TDM_Pickup_Idle) false true))

(fn ConstructionScript
  (|CallParentFunction))
